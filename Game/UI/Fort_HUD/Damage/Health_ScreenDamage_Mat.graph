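# Post-process material for the HUD's health damage effect, blending a pulsing, colored blood-pattern vignette over the scene.
# A screen-space sphere mask confines it to the edges and, with the time-wobbled BloodPattern texture, drives a three-color ramp between black, Color2 and Color1.
# The Opacity parameter and a remapped sine control the final blend.
# The screen-aligned LinearSine branch has no effect because its lerp blends identical texture inputs.
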Material /Game/UI/Fort_HUD/Damage/Health_ScreenDamage_Mat
graph {
 properties {materialDomain=MD_PostProcess, pixelDepthOffsetMode=PDOM_Legacy, shadingModel=MSM_Unlit, usedShadingModels=Unlit}
 nodes [
  n1 = Constant3Vector()
  n2 = VectorParameter() {parameterName=Color2, defaultValue=<rgba 0.25 0.0821 0 1>}
  n3 = VectorParameter() {parameterName=Color1, defaultValue=<rgba 0.5 0.165 0 1>}
  n4 = ComponentMask(None=MaterialFunctionCall() {materialFunction=&ScreenAlignedUVs}."X 100%, Y 100%") {g=true, r=true}
  n5 = SphereMask(A=@n4, B=Constant() {r=0.5}, Radius=ScalarParameter() {defaultValue=0.9548, parameterName=Radius}) {attenuationRadius=0.7, hardnessPercent=0}
  n6 = TextureSample(UVs=Multiply(A=@n4, B=Constant2Vector() {g=2.8125, r=3.9713}) {constB=1.125}) {texture=&/BloodPattern}
  n7 = VectorParameter() {parameterName="V Alt, V Init, H Alt, H Init ", defaultValue=<rgba 1.8 -0.1927 1 0.4862>}
  n8 = Multiply(A=@n4, B=AppendVector(A=@n7.R, B=@n7.R))
  n9 = AppendVector(A=@n7.G, B=@n7.G)
  n10 = MaterialFunctionCall(Value=MaterialFunctionCall(f2=Add(A=@n8, B=@n9)) {materialFunction=&AddComponents}.f2) {materialFunction=&LinearSine}
  n11 = Add(A=Sine(None=Time()) {period=0.75})
  n12 = Clamp(None=Add(A=LinearInterpolate(A=@n6.R, B=@n6.R, Alpha=@n10.Linear Sine), B=Multiply(A=@n11) {constB=-0.1}))
  n13 = Clamp(None=LinearInterpolate(A=LinearInterpolate(Alpha=@n5) {constA=-0.1}, B=LinearInterpolate(Alpha=@n5) {constA=0.4}, Alpha=@n12))
  n14 = Power(Base=@n5) {constExponent=0.75}
  n15 = Clamp(None=Multiply(A=OneMinus(None=@n14)) {constB=1.1})
  n16 = OneMinus(None=@n15)
  n17 = Clamp(None=Multiply(A=Multiply(A=@n12, B=@n16)) {constB=10})
  n18 = Time()
  n19 = Constant() {r=0.5}
  n20 = Multiply(A=@n15, B=ScalarParameter() {parameterName=Opacity})
  LinearInterpolate(
   A=ComponentMask(None=SceneTexture() {sceneTextureId=PPI_PostProcessInput0}.Color) {b=true, g=true, r=true}
   B=Desaturation(
    None=MaterialFunctionCall(1=@n1, 2=@n2.RGB, 3=@n3.RGB, A=@n13) {materialFunction=&Lerp_Multiple_Float3}.Lerp 3 Inputs
    Fraction=Multiply(A=@n17) {constB=0.1}
   )
   Alpha=Multiply(
    A=MaterialFunctionCall!(Sine Phase=@n18, Value 1, Value 2=@n19) {materialFunction=/Engine/Functions/Engine_MaterialFunctions02/Utility/Sine_Remapped}.Remapped Sine
    B=Clamp(None=@n20)
   )
  ) {constA=1}
 ]
}
